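# Physics asset for the AlmondSplash pickaxe, a cosmetic two-handed melee weapon mesh.
# A single capsule body is bound to the Melee_TwoHanded_Handle bone, giving the whole axe one simplified collision volume.
# It has no constraints, so there is no articulated or jointed simulation between parts.
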PhysicsAsset /BRCosmetics/Weapons/FORT_Melee/Pickaxe_AlmondSplash/Meshes/AlmondSplash_Axe_PhysicsAsset
bodies {Melee_TwoHanded_Handle [{shapeType=Capsule, length=75.5388, radius=29.0529, center=<-3.1942 -1.5799 32.3415>, rotation=<rot -0.0105 -90 -2.0455>}]}
constraints []
